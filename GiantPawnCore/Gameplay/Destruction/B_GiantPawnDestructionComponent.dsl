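; Giant pawn component that destroys buildings touched by the pawn's body, using periodic sphere sweeps between configured skeletal mesh socket pairs.
; A server timer runs each enabled trace at its own frequency, with randomized start offsets to spread load across frames.
; Hit actors go to the game state destruction manager; on success, a multicast makes clients repeat the trace and predict building deaths.
; Sweeps run async or synchronously by setting.
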(event EventBeginPlay
  ;[0 0.1354 0.01125 0.5] BeginPlay: Cache variables and start behavior.
  (|CallParentFunction)
  (Class|FortAsyncActionCurrentPlaylistReady|PlaylistDataReadyAsync
    (:OnReady
      (Setup|CacheDestructionManagerGameStateComponent)
      (bind _mesh (Setup|GetSkeletalMeshComponenttoUpdate))
      (Variables|Transient|SetSkeletalMeshComponent _mesh)
      (Networking|SwitchHasAuthority
        (:Authority
          (if (Variables|Settings|GetbAutoStartSkeletalDestructionTraces)
            (CallFunction|StartSkeletonDestructionTraces)))))))

(event Custom|StartSkeletonDestructionTraces
  ;[0.1449 0.09451 0.349] Skeletal Destruction
  ;;[0.000742 0.15 0.03355 0.5] StartSkeletonDestructionTraces: Start the periodic traces that is run along the skeletal mesh component to figure out the buildings to destroy.
  (Utilities|SwitchHasAuthority
    (:Authority
      (bind _output_get (Variables|Transient|SetSkeletalDestructionUpdateFrequency (ScalableFloat|GetValueatLevel (Variables|Hotfix|GetHotfix_SkeletalDestructionUpdateFrequency))))
      (bind _returnvalue (Utilities|Time|SetTimerbyEvent (EventDispatchers|CreateEvent :!event Custom|UpdateSkeletonDestructionTraces) _output_get true true))
      (Variables|Transient|SetSkeletalDestructionUpdateTimerHandle _returnvalue)
      (SkeletalTrace|RandomizeSkeletalDestructionTraceTimings))))

(event Custom|StopSkeletonDestructionTraces
  ;[0.1449 0.09451 0.349] Skeletal Destruction
  ;;[0.15 0.09153 0.000313 0.5] StopSkeletonDestructionTraces: Stop the skeletal destruction traces.
  (Utilities|SwitchHasAuthority
    (:Authority
      (Utilities|Time|ClearandInvalidateTimerbyHandle (Variables|Transient|GetSkeletalDestructionUpdateTimerHandle)))))

(event Custom|UpdateSkeletonDestructionTraces
  (bind _skeletondestructiontraces (Variables|Settings|GetSkeletonDestructionTraces))
  ;[0.1449 0.09451 0.349] Skeletal Destruction
  ;;[0.09375 0 0.07483 0.5] PerformSkeletonDestructionTraces: Perform traces on the skeletal mesh parts that are ready to and multicast the information to the clients.
  (Utilities|SwitchHasAuthority
    (:Authority
      (for (_ _array_index) (Variables|Settings|GetSkeletonDestructionTraces)
        (if (.Enabled_17_B336F2ED4D24095BAB8B8AAB63C060AD (Utilities|Struct|BreakGiantPawnDestructionTraceData (Utilities|Array|Get(aref) _skeletondestructiontraces _array_index)))
          (bind _structout (Utilities|Struct|SetmembersinGiantPawnDestructionTraceData (Utilities|Array|Get(aref) _skeletondestructiontraces _array_index) (+ (.AccumulatedTime_11_67B3A5D54C597AF26E060288979A7334 (Utilities|Struct|BreakGiantPawnDestructionTraceData (Utilities|Array|Get(aref) _skeletondestructiontraces _array_index))) (Variables|Transient|GetSkeletalDestructionUpdateFrequency))))
          (if (> (.Frequency_7_E08BC3834E1EA4F3EC377FB2F3C5758E (Utilities|Struct|BreakGiantPawnDestructionTraceData _structout)) 0)
            (if (> (.AccumulatedTime_11_67B3A5D54C597AF26E060288979A7334 (Utilities|Struct|BreakGiantPawnDestructionTraceData _structout)) (.Frequency_7_E08BC3834E1EA4F3EC377FB2F3C5758E (Utilities|Struct|BreakGiantPawnDestructionTraceData _structout)))
              (bind _structout_2 (Utilities|Struct|SetmembersinGiantPawnDestructionTraceData (Utilities|Array|Get(aref) _skeletondestructiontraces _array_index) (Math|Float|%25%28Float%29 (.AccumulatedTime_11_67B3A5D54C597AF26E060288979A7334 (Utilities|Struct|BreakGiantPawnDestructionTraceData _structout)) (.Frequency_7_E08BC3834E1EA4F3EC377FB2F3C5758E (Utilities|Struct|BreakGiantPawnDestructionTraceData _structout)))))
              (bind _skeletalmeshcomponent (Variables|Transient|GetSkeletalMeshComponent)
                (:then
                  (Helper|PerformTrace (Transformation|GetSocketLocation _skeletalmeshcomponent (Utilities|Struct|BreakGiantPawnDestructionTraceData _structout_2)) (Transformation|GetSocketLocation _skeletalmeshcomponent (.EndSocket_4_4E8A46734BF0E8FCB08CF9A9FEEF7943 (Utilities|Struct|BreakGiantPawnDestructionTraceData _structout_2))) (.Radius_9_38CAF4B34CB59C8504DFCEA077E04114 (Utilities|Struct|BreakGiantPawnDestructionTraceData _structout_2)) (Variables|Settings|GetDestructionTraceObjectTypes) (Helper|GetTraceActorstoIgnore) _array_index (select (Debug|IsDebugEnabeld) "ForDuration" "None") (.DebugColor_15_161A412F4E019BD0B486CA8F5226BA25 (Utilities|Struct|BreakGiantPawnDestructionTraceData _structout_2)) "(R=0.000000,G=1.000000,B=0.000000,A=1.000000)" (.Frequency_7_E08BC3834E1EA4F3EC377FB2F3C5758E (Utilities|Struct|BreakGiantPawnDestructionTraceData _structout_2))))
                (:else)))))))))

(event Destruction|EventHandleAsyncSphereSweepByObjectTypesComplete
  (params (HitResults :type Array%20of%20Hit%20Result%20Structures) (DestructionTraceIndex :type Integer))
  ;[0.1449 0.09451 0.349] Skeletal Destruction
  ;;[0.02583 0.1106 0.15 0.5] HandleSphereSweepTraceComplete: Callback function for when the sweep trace is complete.
  (Utilities|SwitchHasAuthority
    (:Authority
      (bind _actorstodestroy (Helper|GatherActorsfromHitResults HitResults))
      (bind _destructionmanagergamestatecomponent (Variables|Transient|GetDestructionManagerGameStateComponent)
        (:then
          (bind _bvalid (Public|AddBuildingsforDestruction _destructionmanagergamestatecomponent _actorstodestroy (Variables|Settings|GetDestructionLogReason) (Components|GetOwner)))
          (if _bvalid
            (CallFunction|MulticastPredictDestroyBuildingActors DestructionTraceIndex)))
        (:else)))
    (:Remote
      (for _array_element HitResults
        (bind _asbuilding_static_mesh (Utilities|Casting|CastToBuildingSMActor (.HitActor (Collision|BreakHitResult _array_element))))
        (if (Variables|/Script/Engine.Actor|GetbCanBeDamaged _asbuilding_static_mesh)
          (if (> (Math|Vector|VectorLengthSquared (.BoxExtent (Collision|GetActorBounds _asbuilding_static_mesh true false))) (ScalableFloat|GetCachedScalableFloatValue (Variables|Hotfix|GetHotfix_BuildingBoundsSqrToPredictDeath)))
            (PredictiveDestruction|PredictDeathforBuildingActorAsync _asbuilding_static_mesh (Utilities|Casting|CastToFortPawn (Components|GetOwner)) true)
            (else
              (PredictiveDestruction|SimplePredictDeathforBuildingActor _asbuilding_static_mesh))))))))

(event Custom|MulticastPredictDestroyBuildingActors
  (params (DestructionTraceIndex :type Integer))
  (bind (_startsocket_2_09a8aa2343893fd4a6b4488b459a9d80 _endsocket_4_4e8a46734bf0e8fcb08cf9a9feef7943 _frequency_7_e08bc3834e1ea4f3ec377fb2f3c5758e _radius_9_38caf4b34cb59c8504dfcea077e04114) (Utilities|Struct|BreakGiantPawnDestructionTraceData (Utilities|Array|Get(aref) (Variables|Settings|GetSkeletonDestructionTraces) DestructionTraceIndex)))
  (bind _skeletondestructiontraces (Variables|Settings|GetSkeletonDestructionTraces))
  ;[0.1449 0.09451 0.349] Skeletal Destruction
  ;;[0.07596 0.15 0.002807 0.5] MulticastPredictDestroyBuildingActors: On the clients, predict destroy the input buildling actors and update the centroid locations so that it can be read in the gameplay cue.
  (Networking|CanExecuteCosmeticEvents
    (:True
      (if (Utilities|Array|IsValidIndex _skeletondestructiontraces DestructionTraceIndex)
        (bind _skeletalmeshcomponent (Variables|Transient|GetSkeletalMeshComponent)
          (:then
            (Helper|PerformTrace :Start (Transformation|GetSocketLocation _skeletalmeshcomponent _startsocket_2_09a8aa2343893fd4a6b4488b459a9d80) :End (Transformation|GetSocketLocation _skeletalmeshcomponent _endsocket_4_4e8a46734bf0e8fcb08cf9a9feef7943) :Radius _radius_9_38caf4b34cb59c8504dfcea077e04114 :ObjectTypes (Variables|Settings|GetDestructionTraceObjectTypes) :ActorsToIgnore (Helper|GetTraceActorstoIgnore) :DestructionTraceIndex DestructionTraceIndex :DrawDebugType "None" :TraceHitColor "(R=0.000000,G=1.000000,B=0.000000,A=1.000000)" :DrawTime 0.0))
          (:else))))))

(fn CacheDestructionManagerGameStateComponent
  (bind _returnvalue (Fortnite|Athena|GameState|GetGameStateAthena))
  (Utilities|IsValid _returnvalue
    (:Is%20Valid
      (Variables|Transient|SetDestructionManagerGameStateComponent (Actor|GetComponentbyClass _returnvalue "/GiantPawnCore/Gameplay/Destruction/B_GiantPawnDestructionManagerGameStateComponent.B_GiantPawnDestructionManagerGameStateComponent_C")))))

(fn GatherActorsFromHitResults
  (params (HitResults :type Array%20of%20Hit%20Result%20Structures))
  (Utilities|FlowControl|Sequence
    (:then_0
      (for _array_element HitResults
        (Utilities|Array|Add (LocalVariables|GetReturnActorsToDestroy) (.HitActor (Collision|BreakHitResult _array_element)))))
    (:then_1
      (return (LocalVariables|GetReturnActorsToDestroy)))))

(fn GetSkeletalMeshComponentToUpdate
  (Utilities|FlowControl|Sequence
    (:then_0
      (bind _ascharacter (Utilities|Casting|CastToCharacter (Components|GetOwner)))
      (return (Variables|/Script/Engine.Character|GetMesh _ascharacter)))
    (:then_1
      (return 0))))

(fn GetTraceActorsToIgnore
  ;[0.04815 0.15 0.1342 0.5] Return owner actor and attached actors.
  (Utilities|FlowControl|Sequence
    (:then_0
      (bind _returnvalue (Components|GetOwner))
      (Utilities|IsValid _returnvalue
        (:Is%20Valid
          (bind _output_get (LocalVariables|SetRetArray (Actor|GetAttachedActors _returnvalue true false)))
          (Utilities|Array|Add _output_get _returnvalue))))
    (:then_1
      (return (LocalVariables|GetRetArray)))))

(fn IsDebugEnabeld
  (bind _returnvalue (Development|GetConsoleVariableBoolValue "Fort.Giant.DestructionDebugVisualization.Enabled"))
  (return (or _returnvalue (Variables|Debug|GetbDebugEnabled))))

(fn PerformTrace
  (params (Start :type Vector%20%28by%20ref%29) (End :type Vector%20%28by%20ref%29) (Radius :type Float%20%28single-precision%29) (ObjectTypes :type Array%20of%20EObjectTypeQuery%20Enums :default "ObjectTypeQuery1") (ActorsToIgnore :type Array%20of%20Actor%20Object%20References) (DestructionTraceIndex :type Integer) (DrawDebugType :type EDrawDebugTrace%20Enum) (TraceColor :type Linear%20Color%20Structure) (TraceHitColor :type Linear%20Color%20Structure) (DrawTime :type Float%20%28single-precision%29))
  ;[0.02398 0.125 0.15 0.5] PerformTrace: Perform the desired type of trace.
  (if (Variables|Settings|GetbPerformAsyncTraces)
    (Destruction|AsyncSphereSweepbyObjectTypes Start End Radius ObjectTypes false ActorsToIgnore DestructionTraceIndex DrawDebugType TraceColor TraceHitColor DrawTime)
    (else
      (bind _outhits (Collision|MultiSphereTraceForObjects Start End Radius ObjectTypes false ActorsToIgnore DrawDebugType false TraceColor TraceHitColor DrawTime))
      (Destruction|HandleAsyncSphereSweepByObjectTypesComplete _outhits DestructionTraceIndex))))

(fn RandomizeSkeletalDestructionTraceTimings
  (Networking|SwitchHasAuthority
    (:Authority
      (for (_ _array_index) (Variables|Settings|GetSkeletonDestructionTraces)
        (Utilities|Struct|SetmembersinGiantPawnDestructionTraceData (Utilities|Array|Get(aref) (Variables|Settings|GetSkeletonDestructionTraces) _array_index) (Math|Random|RandomFloatinRange 0.0 (.Frequency_7_E08BC3834E1EA4F3EC377FB2F3C5758E (Utilities|Struct|BreakGiantPawnDestructionTraceData (Utilities|Array|Get(aref) (Variables|Settings|GetSkeletonDestructionTraces) _array_index))))) ; this scrambles the accumulated times\nso we don't have multiple things with same frequeny\ntracing on same frames.
      ))))
