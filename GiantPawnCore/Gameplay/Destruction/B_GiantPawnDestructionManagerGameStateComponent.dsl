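; Server-side GameState component that queues building actors for destruction, named for giant pawns, and destroys them in throttled batches.
; Add functions take actors or hit results, reject immune, already-marked, undamageable or non-building actors, then tag and queue the rest, returning their centroids.
; Once playlist data is ready, a looping timer with hotfixable frequency and batch size drains the queue, killing buildings with silent-detachment death credited to the instigator.
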(event EventBeginPlay
  ;[0 0.1615 0.01674] BeginPlay: Cache hotfixes.
  (|CallParentFunction)
  (Utilities|SwitchNetRole
    (:Server
      (Class|FortAsyncActionCurrentPlaylistReady|PlaylistDataReadyAsync
        (:OnReady
          (Variables|Transient|SetNumDestroysPerUpdate (ScalableFloat|GetValueasInteger (Variables|Hotfix|GetHotfix_NumDestroysPerUpdate)))
          (bind _returnvalue (Utilities|Time|SetTimerbyEvent (EventDispatchers|CreateEvent :!event ExecuteDestroyUpdate) (ScalableFloat|GetValueatLevel (Variables|Hotfix|GetHotfix_DestructionUpdateFrequency)) true true))
          (Variables|Transient|SetDestructionUpdateTimerHandle _returnvalue))))))

(event EventEndPlay
  (params (EndPlayReason :type EEndPlayReason%20Enum :default "Destroyed"))
  ;[0.1198 0.07735 0 0.5] EndPlay: Clean up.
  (Utilities|FlowControl|Sequence
    (:then_0
      (Utilities|Time|ClearandInvalidateTimerbyHandle (Variables|Transient|GetDestructionUpdateTimerHandle)))
    (:then_1
      (|CallParentFunction EndPlayReason))))

(event Custom|EnableGiantPawnDestructionGSManagerLog
  (params (bEnable :type Boolean))
  (Variables|Default|SetbDebugLog bEnable))

(event Custom|DebugGiantPawnDestructionManagerSetSkipTagChecks
  (params (bSkip :type Boolean))
  (Utilities|SwitchNetRole
    (:Server
      (Variables|Transient|Debug|SetbDebugSkipTagChecks bSkip))))

(fn AddBuildingsForDestruction
  (params (Actors :type Array%20of%20Actor%20Object%20References) (DestructionLogReason :type String) (DestructionInstigator :type Actor%20Object%20Reference))
  ;[0.03373 0.1418 0.15 0.5] AddBuildingsForDestruction: Add the input building actors for destruction.
  (Utilities|SwitchNetRole
    (:Server
      (for _array_element Actors
        (bind (_candestroy _asbuildingactor _failreason) (Helper|IsActorValidforDestruction _array_element))
        (Utilities|DebugLog (Utilities|String|Append "AddBuildingsForDestruction: Input actor is valid for destruction? = " (Utilities|String|ToString(Boolean) _candestroy) " ; FailReason = " _failreason " ; Actor = " (Utilities|GetDisplayName _array_element) " ; Destruction Log Reason: " DestructionLogReason))
        (if _candestroy
          (Utilities|Array|Add (Variables|Transient|GetDestructionRequestEntries) (Utilities|Struct|MakeGiantPawnDestructionEntry (Utilities|ResolveSoftReference _array_element) DestructionLogReason (Utilities|ResolveSoftReference DestructionInstigator)))
          (Utilities|Array|Add (Variables|/Script/Engine.Actor|GetTags _asbuildingactor) (Variables|Settings|GetMarkedForDestructionActorTag))
          (Utilities|Array|Add (LocalVariables|GetRetBuildingActorCentroidLocations) (Building|GetCentroid _asbuildingactor))
          (Utilities|Array|Add (LocalVariables|GetRetDestroyedBuildingActors) _asbuildingactor)
          (LocalVariables|SetbRetValid true)
          (Utilities|DebugLog (Utilities|String|Append "AddBuildingsForDestruction: Actor is marked for destruction: " (Utilities|GetDisplayName _asbuildingactor)))))
      (return (LocalVariables|GetbRetValid) (LocalVariables|GetRetBuildingActorCentroidLocations) (LocalVariables|GetRetDestroyedBuildingActors)))))

(fn AddBuildingsForDestruction_HitResults
  (params (HitResults :type Array%20of%20Hit%20Result%20Structures) (DestructionLogReason :type String) (DestructionInstigator :type Actor%20Object%20Reference))
  ;[0.03373 0.1418 0.15 0.5] AddBuildingsForDestruction: Add the input building actors for destruction.
  (Utilities|SwitchNetRole
    (:Server
      (for _array_element HitResults
        (bind _asbuilding_actor (Utilities|Casting|CastToBuildingActor (.HitActor (Collision|BreakHitResult _array_element))))
        (Utilities|Array|AddUnique (LocalVariables|GetBuildingsForDestruction) _asbuilding_actor))
      (bind (_bvalid _destroyedbuildingactorscentroidlocations _destroybuildingactors) (Public|AddBuildingsforDestruction (LocalVariables|GetBuildingsForDestruction) DestructionLogReason DestructionInstigator))
      (return _bvalid _destroyedbuildingactorscentroidlocations _destroybuildingactors))
    (:Client
      (return false 0 0))))

(fn ExecuteDestroyUpdate
  (bind _destructionrequestentries (Variables|Transient|GetDestructionRequestEntries))
  (bind _output (Utilities|Array|Get(aref) _destructionrequestentries))
  (bind (_actortodestroy_12_fde5f820464fb76bab2f09865cea02f5 _destructionlogreason_7_cb76d104446cdde11b95ad91a33ce050 _destructioninstigator_11_c8e3c92a4b34952d8d685e947f7d7e44) (Utilities|Struct|BreakGiantPawnDestructionEntry _output))
  (Utilities|SwitchNetRole
    (:Server
      (bind _output_get (LocalVariables|SetNumOfDestroysPending (Variables|Transient|GetNumDestroysPerUpdate)))
      (Utilities|DebugLog (Utilities|String|Append "ExecuteDestroyUpdate: Invoked with Destroys pending: " (Utilities|String|ToString(Integer) _output_get)))
      (while (and (> (LocalVariables|GetNumOfDestroysPending) 0) (Utilities|Array|IsNotEmpty (Variables|Transient|GetDestructionRequestEntries)))
        (if (Utilities|Array|IsValidIndex _destructionrequestentries)
          (Utilities|FlowControl|Sequence
            (:then_0
              (bind _asbuilding_actor (Utilities|Casting|CastToBuildingActor (Utilities|ResolveSoftReference _actortodestroy_12_fde5f820464fb76bab2f09865cea02f5)))
              (Utilities|IsValid _asbuilding_actor
                (:Is%20Valid
                  (if (not (Variables|/Script/FortniteGame.BuildingActor|GetbDestroyed _asbuilding_actor))
                    (Gameplay|Die _asbuilding_actor "(TagName=\"Gameplay.Effect.InstantDeath.SilentDetachment\")" (Fortnite|GetFortPlayerControllerfromActor (Utilities|ResolveSoftReference _destructioninstigator_11_c8e3c92a4b34952d8d685e947f7d7e44)) (Utilities|ResolveSoftReference _destructioninstigator_11_c8e3c92a4b34952d8d685e947f7d7e44) :!class /Script/FortniteGame.BuildingActor)
                    (Math|Integer|DecrementInt (LocalVariables|GetNumOfDestroysPending)
                      (:%20%20
                        (Utilities|DebugLog (Utilities|String|Append "ExecuteDestroyUpdate: Destroying building " (Utilities|GetDisplayName _asbuilding_actor) " with reason " _destructionlogreason_7_cb76d104446cdde11b95ad91a33ce050))))))))
            (:then_1
              (Utilities|Array|RemoveIndex (Variables|Transient|GetDestructionRequestEntries)))))))))

(fn ForceExecuteDestroyUpdate
  ;[0 0.15 0.0883 0.5] Force execute the destroy update.
  (Utilities|SwitchNetRole
    (:Server
      (Protected|ExecuteDestroyUpdate))))

(fn IsActorValidForDestruction
  (params (TestActor :type Actor%20Object%20Reference))
  ;[0.04861 0.1354 0.1321 0.5] IsActorValidForDestruction: Returns if the actor can be destroyed.
  (Utilities|IsValid TestActor
    (:Is%20Valid
      (if (not (Variables|Transient|Debug|GetbDebugSkipTagChecks))
        (if (not (Actor|ActorHasTag (Variables|GetTestActor) (Variables|Settings|GetMarkedForDestructionActorTag)))
          (if (Actor|ActorHasTag (Variables|GetTestActor) (Variables|Settings|GetDestructionImmuneTag))
            (return false 0 "Actor has DestructionImmune tag"))
          (else
            (return false 0 "Actor has MarkedForDestruction tag"))))
      (bind _asbuilding_actor (Utilities|Casting|CastToBuildingActor (Variables|GetTestActor))
        (:then
          (if (Variables|/Script/Engine.Actor|GetbCanBeDamaged _asbuilding_actor)
            (return true _asbuilding_actor "")
            (else
              (return false 0 "Test actor cannot be damaged."))))
        (:CastFailed
          (return false 0 "Test actor is not a buildling"))))
    (:Is%20Not%20Valid
      (return false 0 "TestActor is Invalid"))))
